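# Unlit translucent material for Niagara mesh particles, evidently the water splash effect of the LilSplit glider.
# A glass reflection texture sampled by reflection vector is contrast-boosted, added to particle color and passed through an emissive adjustment function.
# A narrow SmoothStep creates a hard-edged erosion mask from noise offset by the Erode dynamic parameter, particle alpha and a gradient of pixel height relative to the particle position.
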Material /BRCosmetics/Gadgets/Assets/VinderTech_GliderChute/Glider_LilSplit/FX/M_LilSplit_WaterSplash
graph {
 properties {
  bAutomaticallySetUsageInEditor=false
  bTangentSpaceNormal=false
  bUsedWithNiagaraMeshParticles=true
  blendMode=BLEND_Translucent
  pixelDepthOffsetMode=PDOM_Legacy
  shadingModel=MSM_Unlit
  usedShadingModels=Unlit
 }
 nodes [
  n1 = TextureSample(UVs=ComponentMask(None=ReflectionVectorWS()) {g=true, r=true}) {texture=/Game/Items/Art_noLOD/Textures/T_Items_MineralPowder_Glass_Reflection}
  n2 = ParticleColor()
  n3 = Add(
   A=MaterialFunctionCall(In=@n1.G, Contrast=Constant() {r=2}) {materialFunction=&CheapContrast}.Result
   B=@n2.RGB
  )
  MaterialFunctionCall(Emissive=@n3) {materialFunction=&MF_FX_EmissiveAdjustments}
  n4 = MaterialFunctionCall(Float3=WorldPosition().XYZ) {materialFunction=&BreakOutFloat3Components}
  n5 = ParticlePositionWS()
  n6 = MaterialFunctionCall(Float3=@n5) {materialFunction=&BreakOutFloat3Components}
  n7 = Clamp(None=Divide(A=Subtract(A=@n4.B, B=@n6.B)) {constB=35})
  Multiply(A=@n3, B=@n7)
  Multiply(A=@n1.G, B=@n2.A)
  Multiply(A=VertexNormalWS(), B=Constant() {r=1.1})
  ScalarParameter() {parameterName=Z SpawnHeight}
  ScalarParameter() {parameterName=alphatest}
  n8 = Add(
   A=TextureSample() {samplerType=SAMPLERTYPE_LinearColor, texture=/Game/Global/Textures/Default/Noise/T_Noise01}.G
   B=DynamicParameter() {paramNames [Erode,Param2,Param3,Param4]}.Erode
  )
  SmoothStep(Value=Multiply(A=Multiply(A=@n2.A, B=@n7), B=Clamp(None=@n8))) {constMax=0.25, constMin=0.2}
  Transform(None=@n5) {transformSourceType=TRANSFORMSOURCE_Local, transformType=TRANSFORM_Instance}
 ]
}
