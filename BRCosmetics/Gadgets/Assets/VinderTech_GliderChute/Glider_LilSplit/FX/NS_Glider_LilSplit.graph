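# Niagara effect for the LilSplit glider cosmetic, trailing droplets, small tumbling bits, ribbons and speed lines.
# Sprites, attached ribbons and bone-pair anim trails follow the user skeletal mesh, with droplets simulated on GPU.
# VariantID picks color palettes, ForwardDot modulates spawning and motion, and owner speed scales long-ribbon velocity.
# Front-end switches adjust values for preview, time-of-day helpers tint several layers, and one sprite/mesh emitter is disabled.
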NiagaraSystem /BRCosmetics/Gadgets/Assets/VinderTech_GliderChute/Glider_LilSplit/FX/NS_Glider_LilSplit
dependencies {
 usedDataInterfaces [
  &NiagaraDataInterfaceArrayColor
  &NiagaraDataInterfaceVector2DCurve
  &NiagaraDataInterfaceCurve
  &NiagaraDataInterfaceColorCurve
  &NiagaraDataInterfaceArrayInt32
 ]
 usedDynamicInputs [&Multiply_Vector2D_ByFloat, &Lerp_Vector2]
 usedModules [&SystemState, &Timeline]
}
emitters {
 Drops {
  simTarget=GPUComputeSim
  emitterUpdateScript [
   &EmitterState(Life Cycle Mode=::NewEnumerator1, Loop Delay=$User.AnimationDelay, Loop Duration=3, UseLoopDelay=false)
   &SpawnRate(
    Spawn Probability=0.55
    SpawnRate=&Lerp_Float(
     A=&FN_FrontEndFloatModifier(Gameplay=8, Front End=80)
     B=55
     Alpha=&ClampFloat(Float=$User.ForwardDot)
    )
    Use Spawn Probability=true
   )
  ]
  particleSpawnScript [
   &FN_SetSpawnGroup(Number of Spawn Groups=2)
   &InitializeParticle(
    Color=&MakeCustomLinearColorFromBool(
     Bool=&SetBoolByFloatComparison(A=$User.VariantID, B=1, Comparison Type=::NewEnumerator6)
     True Color=&SelectLinearColorFromArray(
      Array Sampling Mode=::NewEnumerator0
      Direct Array Index=$Particles.SpawnGroup
      Color Selection Array=DataInterface() {propertyValues {ColorData [<rgba 0.1667 0.0516 0.0333 1>, <rgba 0.8802 0.4621 0.2567 1>]}}
      Random Seed=0
     )
     False Color=&MakeCustomLinearColorFromBool(
      Bool=&SetBoolByFloatComparison(A=$User.VariantID, B=2, Comparison Type=::NewEnumerator6)
      True Color=&SelectLinearColorFromArray(
       Array Sampling Mode=::NewEnumerator0
       Direct Array Index=&SpawnGroup()
       Color Selection Array=DataInterface() {propertyValues {ColorData [<rgba 0.1927 0.0435 0.0569 1>, <rgba 0.8802 0.33 0.3388 1>]}}
       Random Seed=0
      )
      False Color=&SelectLinearColorFromArray(
       Array Sampling Mode=::NewEnumerator0
       Direct Array Index=&SpawnGroup()
       Color Selection Array=DataInterface() {propertyValues {ColorData [<rgba 0.4097 0.8125 0.2182 1>, <rgba 0.0708 0.1719 0.0144 1>]}}
       Random Seed=0
      )
     )
    )
    Lifetime Max=0.6
    Lifetime Min=0.3
    Lifetime Mode=::NewEnumerator1
    Position Offset=<0 0 -100>
    Sprite Size Mode=::NewEnumerator1
    Uniform Sprite Size Max=18
    Uniform Sprite Size Min=10
    UsePositionOffset=false
    Write Material Random=true
   )
   &SkeletalMeshLocation(
    Bone / Socket Index=$Particles.SpawnGroup
    Bone Sampling Mode=::NewEnumerator4
    MaskSpawnGroup=false
    Sampled Position Offset=<12 0 0>
    Skeletal Mesh=$User.SkeletalMesh
   )
   &ShapeLocation(MaskSpawnGroup=false, Override Seed=false, Sphere Radius=20, UseOwnerScale=true) {enabled=false}
   &AddVelocity(
    Velocity=&Multiply_Vector(
     A=&RandomRangeVector(Minimum=<100 200 100>, Maximum=<-100 300 -100>)
     B=&FN_FrontEndVector3Modifier(Gameplay=<1 1 1>, Front End=<1 -1 1>)
    )
    Velocity Speed Scale=HlslExpression() {hlslExpression="User.ForwardDot > -0.5 ? 1 : 0.1"}
   )
  ]
  particleUpdateScript [
   &ParticleState()
   &ScaleSpriteSize(
    Scale Sprite Size Mode=::NewEnumerator0
    Uniform Scale Factor=&FloatFromCurve(FloatCurve=DataInterface() {propertyValues {Curve {keys [
         <at 0 0.75>
         {interpMode=RCIM_Cubic, leaveTangent=0.34, tangentMode=RCTM_Break, time=0.2, value=0.95}
         {arriveTangent=-2.5103, interpMode=RCIM_Cubic, tangentMode=RCTM_Break, time=1, value=0.15}
        ]}}})
   )
   &GravityForce(Gravity=&Lerp_Vector(A=<0 0 -400>, B=<0 0 -200>, Alpha=$User.ForwardDot))
   &Drag(Drag=0.5, Use Linear Drag=true, Use Rotational Drag=true)
   &SolveForcesAndVelocity(Clamp Velocity=false, Limit Acceleration=false)
   &ScaleSpriteSizeBySpeed(Min Scale Factor=<0.6 1.5>)
   &DynamicMaterialParameters(
    Index 0 Param 1=&RandomRangeFloat(Minimum=0, Maximum=1)
    Index 0 Param 2=0.1
    Index 0 Param 3=1
    Index 0 Param 4=&FloatFromCurve(FloatCurve=DataInterface() {propertyValues {Curve {keys [
         {interpMode=RCIM_Cubic, leaveTangent=-0.1723, tangentMode=RCTM_Break, time=0, value=0}
         <at 0.4 0.7>
         {arriveTangent=0.0065, interpMode=RCIM_Cubic, tangentMode=RCTM_Break, time=1, value=1}
        ]}}})
    Module.Index0Param0Write=true
    Module.Index0Param1Write=false
    Module.Index0Param2Write=true
    Module.Index0Param3Write=true
   )
   &ScaleColor(
    Scale Alpha=&Multiply_Float(
     A=&FNTimeOfDayScalar(Night Scale=1.5, Morning Scale=2, Evening Scale=2)
     B=&FloatFromCurve(FloatCurve=DataInterface() {propertyValues {Curve {keys [<at 0 0.625>,<at 0.2 0.95>,<at 0.8 0.95>,<at 1 0.75>]}}})
    )
    Scale RGB=&Multiply_Vector(
     A=&MakeVectorFromLinearColorRGB(Color=<rgba 1 1 1 1>)
     B=&VectorFromFloat(Value=1.2)
    )
    ScaleA=true
    ScaleRGB=true
   )
   &Color(Color=&MultiplyLinearColor(
     Linear Color=$Particles.Color
     Scale Factor=&MakeVector4_FromLinearColor(LinearColor=&FNTimeOfDayColor(
       Night Color=<rgba 0.5104 0.2772 0.4449 0.571>
       Morning Color=<rgba 1 1 1 1>
       Day Color=<rgba 1 1 1 1>
       Evening Color=<rgba 0.9688 0.9688 0.9688 1>
      ))
    ))
  ]
  rendererClasses [&NiagaraSpriteRendererProperties]
  renderers [{rendererClass=&NiagaraSpriteRendererProperties}]
 }
 Drops001 {
  simTarget=GPUComputeSim
  emitterUpdateScript [
   &EmitterState(Life Cycle Mode=::NewEnumerator1, Loop Delay=$User.AnimationDelay, Loop Duration=3, UseLoopDelay=false)
   &SpawnRate(Spawn Probability=0.4, SpawnRate=3, Use Spawn Probability=true)
  ]
  particleSpawnScript [
   &FN_SetSpawnGroup(Number of Spawn Groups=2)
   &InitializeParticle(
    Color=&MakeCustomLinearColorFromBool(
     Bool=&SetBoolByFloatComparison(A=$User.VariantID, B=1, Comparison Type=::NewEnumerator6)
     True Color=&SelectLinearColorFromArray(
      Array Sampling Mode=::NewEnumerator0
      Direct Array Index=?
      Color Selection Array=DataInterface() {propertyValues {ColorData [<rgba 0.1667 0.0516 0.0333 1>, <rgba 0.8802 0.4621 0.2567 1>]}}
      Random Seed=0
     )
     False Color=&MakeCustomLinearColorFromBool(
      Bool=&SetBoolByFloatComparison(A=$User.VariantID, B=2, Comparison Type=::NewEnumerator6)
      True Color=&SelectLinearColorFromArray(
       Array Sampling Mode=::NewEnumerator0
       Direct Array Index=?
       Color Selection Array=DataInterface() {propertyValues {ColorData [<rgba 0.1927 0.0435 0.0569 1>, <rgba 0.8802 0.33 0.3388 1>]}}
       Random Seed=0
      )
      False Color=&SelectLinearColorFromArray(
       Array Sampling Mode=::NewEnumerator0
       Direct Array Index=?
       Color Selection Array=DataInterface() {propertyValues {ColorData [<rgba 0.4097 0.8125 0.2182 1>, <rgba 0.0708 0.1719 0.0144 1>]}}
       Random Seed=0
      )
     )
    )
    Lifetime Max=1
    Lifetime Min=0.3
    Lifetime Mode=::NewEnumerator1
    Position Offset=<0 0 -100>
    Sprite Size Mode=::NewEnumerator1
    Uniform Sprite Size Max=15
    Uniform Sprite Size Min=10
    UsePositionOffset=false
    Write Material Random=true
   )
   &SkeletalMeshLocation(
    Bone / Socket Index=&SelectIntFromArray(Direct Array Index=$Particles.SpawnGroup, Int Selection Array=DataInterface() {propertyValues {IntData [6,7]}})
    MaskSpawnGroup=false
    Mesh Sampling Type=::NewEnumerator5
    Skeletal Mesh=$User.SkeletalMesh
    Socket Sampling Mode=::NewEnumerator5
   )
   &ShapeLocation(MaskSpawnGroup=false, Override Seed=false, Sphere Radius=20, UseOwnerScale=true) {enabled=false}
   &AddVelocity(
    Velocity=&RandomRangeVector(Minimum=<80 250 50>, Maximum=<-160 350 -100>)
    Velocity Speed Scale=HlslExpression() {hlslExpression="User.ForwardDot > -0.5 ? 1 : 0.1"}
   ) {enabled=false}
  ]
  particleUpdateScript [
   &ParticleState()
   &ScaleSpriteSize(
    Scale Sprite Size Mode=::NewEnumerator0
    Uniform Scale Factor=&FloatFromCurve(FloatCurve=DataInterface() {propertyValues {Curve {keys [
         <at 0 0.75>
         {interpMode=RCIM_Cubic, leaveTangent=0.34, tangentMode=RCTM_Break, time=0.2, value=0.95}
         {arriveTangent=-2.5103, interpMode=RCIM_Cubic, tangentMode=RCTM_Break, time=1, value=0.15}
        ]}}})
   )
   &GravityForce(Gravity=&Lerp_Vector(A=<0 0 -500>, B=<0 0 -200>, Alpha=$User.ForwardDot))
   &Drag(Drag=0.5, Use Linear Drag=true, Use Rotational Drag=true)
   &SolveForcesAndVelocity(Clamp Velocity=false, Limit Acceleration=false)
   &ScaleSpriteSizeBySpeed(Min Scale Factor=<0.6 1.5>)
   &DynamicMaterialParameters(
    Index 0 Param 1=&RandomRangeFloat(Minimum=0, Maximum=1)
    Index 0 Param 2=0.1
    Index 0 Param 3=1
    Index 0 Param 4=&FloatFromCurve(FloatCurve=DataInterface() {propertyValues {Curve {keys [
         {interpMode=RCIM_Cubic, leaveTangent=-0.1723, tangentMode=RCTM_Break, time=0, value=0}
         <at 0.4 0.7>
         {arriveTangent=0.0065, interpMode=RCIM_Cubic, tangentMode=RCTM_Break, time=1, value=1}
        ]}}})
    Module.Index0Param0Write=true
    Module.Index0Param1Write=false
    Module.Index0Param2Write=true
    Module.Index0Param3Write=true
   )
   &ScaleColor(
    Scale Alpha=&Multiply_Float(
     A=&FNTimeOfDayScalar(Night Scale=1.5, Morning Scale=2, Evening Scale=2)
     B=&FloatFromCurve(FloatCurve=DataInterface() {propertyValues {Curve {keys [<at 0 0.4>,<at 0.2 0.95>,<at 1 0.95>]}}})
    )
    Scale RGB=&Multiply_Vector(
     A=&MakeVectorFromLinearColorRGB(Color=<rgba 1 1 1 1>)
     B=&VectorFromFloat(Value=1.2)
    )
    ScaleA=true
    ScaleRGB=true
   )
   &Color(Color=&MultiplyLinearColor(
     Linear Color=$Particles.Color
     Scale Factor=&MakeVector4_FromLinearColor(LinearColor=&FNTimeOfDayColor(
       Night Color=<rgba 0.5104 0.2772 0.4449 0.571>
       Morning Color=<rgba 1 1 1 1>
       Day Color=<rgba 1 1 1 1>
       Evening Color=<rgba 0.9688 0.9688 0.9688 1>
      ))
    ))
  ]
  rendererClasses [&NiagaraSpriteRendererProperties]
  renderers [{rendererClass=&NiagaraSpriteRendererProperties}]
 }
 Empty {
  emitterUpdateScript [
   &EmitterState(Life Cycle Mode=::NewEnumerator1, Loop Delay=$User.AnimationDelay, Loop Duration=3, UseLoopDelay=false)
   &SpawnBurst_Instantaneous(
    Spawn Count=&RandomRangeInt(Minimum=1, Maximum=5)
    Spawn Probability=0.5
    Use Spawn Probability=true
    UseLoopCountLimit=false
   ) {enabled=false}
   &SpawnRate(
    Spawn Probability=&Lerp_Float(A=0.3, B=0.6, Alpha=$User.ForwardDot)
    SpawnRate=&Lerp_Float(A=10, B=50, Alpha=$User.ForwardDot)
    Use Spawn Probability=true
   )
  ]
  particleSpawnScript [
   &FN_SetSpawnGroup(Number of Spawn Groups=2)
   &InitializeParticle(
    Color=&MakeCustomLinearColorFromBool(
     Bool=&SetBoolByFloatComparison(A=$User.VariantID, B=1, Comparison Type=::NewEnumerator6)
     True Color=&SelectLinearColorFromArray(
      Array Sampling Mode=::NewEnumerator0
      Direct Array Index=?
      Color Selection Array=DataInterface() {propertyValues {ColorData [<rgba 0.8958 0.0559 0.0648 1>, <rgba 1 0.1094 0.7312 1>, <rgba 0.2135 0.7811 1 1>, <rgba 0.1211 0.0913 1 0.995>]}}
      Random Seed=0
     )
     False Color=&MakeCustomLinearColorFromBool(
      Bool=&SetBoolByFloatComparison(A=$User.VariantID, B=2, Comparison Type=::NewEnumerator6)
      True Color=&SelectLinearColorFromArray(
       Array Sampling Mode=::NewEnumerator0
       Direct Array Index=?
       Color Selection Array=DataInterface() {propertyValues {ColorData [<rgba 0.0573 0.016 0.0097 1>, <rgba 0.2656 0.0498 0.0524 1>, <rgba 0.1615 0.0438 0.0048 1>, <rgba 0.1406 0.0154 0.007 0.995>]}}
       Random Seed=0
      )
      False Color=&SelectLinearColorFromArray(
       Array Sampling Mode=::NewEnumerator0
       Direct Array Index=?
       Color Selection Array=DataInterface() {propertyValues {ColorData [<rgba 0.4677 0.6979 0.3344 1>, <rgba 0.1879 0.3333 0.0623 1>, <rgba 0.0647 0.0833 0.0126 1>, <rgba 0.3739 0.4115 0.0755 0.995>]}}
       Random Seed=0
      )
     )
    )
    Lifetime Max=0.6
    Lifetime Min=0.3
    Lifetime Mode=::NewEnumerator1
    Position Offset=<0 0 -15>
    Sprite Rotation Mode=::NewEnumerator2
    Sprite Size=<6 20>
    Sprite Size Mode=::NewEnumerator4
    UsePositionOffset=true
    Write Material Random=true
   )
   &SkeletalMeshLocation(Bone / Socket Index=$Particles.SpawnGroup, Bone Sampling Mode=::NewEnumerator4, MaskSpawnGroup=false, Skeletal Mesh=$User.SkeletalMesh)
   &ShapeLocation(MaskSpawnGroup=false, Override Seed=false, Sphere Radius=20, UseOwnerScale=true)
   &AddVelocity(
    Velocity=&RandomRangeVector(Minimum=<25 30 25>, Maximum=<-25 80 -25>)
    Velocity Speed Scale=&Multiply_Float(
     A=&Lerp_Float(A=3, B=5, Alpha=$User.ForwardDot)
     B=&FN_FrontEndFloatModifier(Gameplay=1, Front End=-1)
    )
   )
  ]
  particleUpdateScript [
   &ParticleState()
   &SpriteRotationRate(Rotation Rate=&RandomRangeFloat(Minimum=-100, Maximum=100))
   &ScaleSpriteSize(Uniform Curve Sprite Scale=DataInterface() {propertyValues {Curve {keys [
        {tangentMode=RCTM_User, time=0, value=0.35}
        {interpMode=RCIM_Cubic, leaveTangent=0.2166, tangentMode=RCTM_Break, time=0.3, value=0.95}
        {arriveTangent=-2.3045, interpMode=RCIM_Cubic, tangentMode=RCTM_Break, time=1, value=0.35}
       ]}}})
   &GravityForce(Gravity=<0 0 -500>)
   &SolveForcesAndVelocity(Clamp Velocity=false, Limit Acceleration=false)
  ]
  rendererClasses [&NiagaraSpriteRendererProperties]
  renderers [{rendererClass=&NiagaraSpriteRendererProperties}]
 }
 Empty001 {
  bEnabled=false
  emitterUpdateScript [
   &EmitterState(Life Cycle Mode=::NewEnumerator1, Loop Delay=$User.AnimationDelay, Loop Duration=3, UseLoopDelay=false)
   &SpawnBurst_Instantaneous(
    Spawn Count=&RandomRangeInt(Minimum=1, Maximum=5)
    Spawn Probability=0.5
    Use Spawn Probability=true
    UseLoopCountLimit=false
   ) {enabled=false}
   &SpawnRate(
    Spawn Probability=&Lerp_Float(A=0.2, B=0.5, Alpha=$User.ForwardDot)
    SpawnRate=&Lerp_Float(A=6, B=12, Alpha=$User.ForwardDot)
    Use Spawn Probability=true
   )
  ]
  particleSpawnScript [
   &FN_SetSpawnGroup(Number of Spawn Groups=2)
   &InitializeParticle(
    Color=<rgba 0.9635 0.6636 0.3111 1>
    Lifetime Max=0.7
    Lifetime Min=0.3
    Lifetime Mode=::NewEnumerator1
    Mesh Scale Mode=::NewEnumerator1
    Mesh Uniform Scale Max=5
    Mesh Uniform Scale Min=3
    Position Offset=<0 0 -15>
    Sprite Rotation Mode=::NewEnumerator2
    Sprite Size Mode=::NewEnumerator1
    Uniform Sprite Size Max=50
    Uniform Sprite Size Min=35
    UsePositionOffset=false
    Write Material Random=true
   )
   &SkeletalMeshLocation(Bone / Socket Index=$Particles.SpawnGroup, Bone Sampling Mode=::NewEnumerator4, MaskSpawnGroup=false, Skeletal Mesh=$User.SkeletalMesh)
   &ShapeLocation(MaskSpawnGroup=false, Override Seed=false, Sphere Radius=20, UseOwnerScale=true)
   &AddVelocity(
    Velocity=&RandomRangeVector(Minimum=<0 20 0>, Maximum=<0 100 0>)
    Velocity Speed Scale=&Lerp_Float(A=1, B=5, Alpha=$User.ForwardDot)
   )
  ]
  particleUpdateScript [
   &ParticleState()
   &ScaleMeshSize(Scale Factor=&VectorFromFloat(Value=&FloatFromCurve(FloatCurve=DataInterface() {propertyValues {Curve {keys [
          {tangentMode=RCTM_User, time=0, value=0.35}
          {interpMode=RCIM_Cubic, leaveTangent=0.2166, tangentMode=RCTM_Break, time=0.3, value=0.95}
          {arriveTangent=-2.3045, interpMode=RCIM_Cubic, tangentMode=RCTM_Break, time=1, value=0.35}
         ]}}})))
   &SpriteRotationRate(Rotation Rate=&RandomRangeFloat(Minimum=-100, Maximum=100))
   &ScaleSpriteSize(Uniform Curve Sprite Scale=DataInterface() {propertyValues {Curve {keys [
        {tangentMode=RCTM_User, time=0, value=0.35}
        {interpMode=RCIM_Cubic, leaveTangent=0.2166, tangentMode=RCTM_Break, time=0.3, value=0.95}
        {arriveTangent=-2.3045, interpMode=RCIM_Cubic, tangentMode=RCTM_Break, time=1, value=0.35}
       ]}}}) {enabled=false}
   &GravityForce(Gravity=<0 0 -500>)
   &SolveForcesAndVelocity(Clamp Velocity=false, Limit Acceleration=false)
  ]
  rendererClasses [&NiagaraSpriteRendererProperties, &NiagaraMeshRendererProperties]
  renderers [
   {rendererClass=&NiagaraSpriteRendererProperties}
   {rendererClass=&NiagaraMeshRendererProperties, rendererIndex=1}
  ]
 }
 FN_AttachedRibbon {
  emitterUpdateScript [
   &EmitterState()
   &FN_SpawnRibbon_v2(Number of Ribbons=2, Spawn Mode=::NewEnumerator0, SpawnRate=30)
   &FN_Integrator() {enabled=false}
  ]
  particleSpawnScript [
   &FN_SetSpawnGroup(Number of Spawn Groups=2)
   /Game/Effects/Niagara/Modules/Ribbons/FN_InitializeRibbon_v4(
    Color:LinearColor=<rgba 0.1615 0.0567 0.039 1>
    Enable Skeletal Position=true
    Lifetime=1
    Offset Amount=-10
    Ribbon Width=45
    SkeletalMesh:NiagaraDataInterfaceSkeletalMesh=$User.SkeletalMesh
   )
   /SetVariables(Particles.AlphaFadeIn:NiagaraFloat=/Niagara/DynamicInputs/Math/Min_Float(A:NiagaraFloat=&Divide_Float(A=$Emitter.Age, B=1), B:NiagaraFloat=1)) {isSetParameters=true}
   &AddVelocity(Velocity=&FN_FrontEndVector3Modifier(Gameplay=<0 400 0>, Front End=<0 -270 0>))
   &FN_FlipRibbonV()
  ]
  particleUpdateScript [
   &SolveForcesAndVelocity(Clamp Velocity=false, Limit Acceleration=false)
   /Game/Effects/Niagara/Modules/Ribbons/FN_UpdateRibbon_v4(Enable Skeletal Position=true, SkeletalMesh:NiagaraDataInterfaceSkeletalMesh=$User.SkeletalMesh)
   &Color(
    Color=&SelectLinearColorFromArray(
     Array Sampling Mode=::NewEnumerator1
     Direct Array Index=&MakeIntFromFloat(Float=$User.VariantID)
     Color Selection Array=DataInterface() {propertyValues {ColorData [<rgba 0.1979 0.042 0.0282 1>, <rgba 0.2292 0.0334 0.0379 1>, <rgba 0.0701 0.1615 0.0508 1>]}}
     Random Seed=?
    )
    Scale Alpha=&Multiply_Float(
     A=&Multiply_Float(A=0.35, B=$Particles.AlphaFadeIn)
     B=&FN_FrontEndFloatModifier(Gameplay=$System.FN_Integrator001.Value, Front End=1)
    )
   )
  ]
  rendererClasses [&NiagaraRibbonRendererProperties]
  renderers [{rendererClass=&NiagaraRibbonRendererProperties}]
 }
 LongerRibbon {
  emitterUpdateScript [
   &EmitterState()
   /Game/Effects/Niagara/Modules/Ribbons/FN_SpawnRibbon_v3(Number of Ribbons=2, Skeleton Mesh:NiagaraDataInterfaceSkeletalMesh=$User.SkeletalMesh, SpawnRate=16)
  ]
  particleSpawnScript [
   &FN_SetSpawnGroup(Number of Spawn Groups=2)
   /Game/Effects/Niagara/Modules/Ribbons/FN_InitializeRibbon_v5(
    Enable Skeletal Position=true
    Lifetime=2
    Offset Amount=15
    Ribbon Width=70
    SkeletalMesh:NiagaraDataInterfaceSkeletalMesh=$User.SkeletalMesh
   )
   &AddVelocity(Velocity=&MakeVector(
     X=&Multiply_Float(A=55, B=&Multiply_Float(A=$User.RightDot, B=-1))
     Y=&FN_FrontEndFloatModifier(Gameplay=270, Front End=-350)
     Z=0
    ))
   /SetVariables(Particles.Color=&MakeCustomLinearColorFromBool(
     Bool=&SetBoolByFloatComparison(A=$User.VariantID, B=1, Comparison Type=::NewEnumerator6)
     True Color=&ColorFromCurve(ColorCurve=DataInterface() {propertyValues {
        AlphaCurve {keys [<at 0 1 RCIM_Cubic>,<at 1 1 RCIM_Cubic>]}
        BlueCurve {keys [<at 0 0.0283 RCIM_Cubic>,<at 0.2068 0.162>,<at 0.4703 0.03>,<at 0.7289 0.1613>,<at 1 0.0666 RCIM_Cubic>]}
        GreenCurve {keys [<at 0 0.0742 RCIM_Cubic>,<at 0.2068 0.2689>,<at 0.4703 0.0494>,<at 0.7289 0.277>,<at 1 0.1119 RCIM_Cubic>]}
        RedCurve {keys [<at 0 0.1979 RCIM_Cubic>,<at 0.2068 0.8229>,<at 0.4703 0.1771>,<at 0.7289 0.5729>,<at 1 0.2747 RCIM_Cubic>]}
       }})
     False Color=&MakeCustomLinearColorFromBool(
      Bool=&SetBoolByFloatComparison(A=$User.VariantID, B=2, Comparison Type=::NewEnumerator6)
      True Color=&ColorFromCurve(ColorCurve=DataInterface() {propertyValues {
         AlphaCurve {keys [<at 0 1 RCIM_Cubic>,<at 1 1 RCIM_Cubic>]}
         BlueCurve {keys [<at 0 0.118 RCIM_Cubic>,<at 0.2068 0.2889>,<at 0.4703 0.2242>,<at 0.7289 0.283>,<at 1 0.1953 RCIM_Cubic>]}
         GreenCurve {keys [<at 0 0.1067 RCIM_Cubic>,<at 0.2068 0.3487>,<at 0.4703 0.2495>,<at 0.7289 0.3541>,<at 1 0.2015 RCIM_Cubic>]}
         RedCurve {keys [<at 0 0.3802 RCIM_Cubic>,<at 0.2068 1>,<at 0.4703 0.6354>,<at 0.7289 0.8906>,<at 1 0.625 RCIM_Cubic>]}
        }})
      False Color=&ColorFromCurve(ColorCurve=DataInterface() {propertyValues {
         AlphaCurve {keys [<at 0 1 RCIM_Cubic>,<at 1 1 RCIM_Cubic>]}
         BlueCurve {keys [<at 0 0.036 RCIM_Cubic>,<at 0.2068 0.0719>,<at 0.4703 0.0535>,<at 0.7289 0.1597>,<at 1 0.0441 RCIM_Cubic>]}
         GreenCurve {keys [<at 0 0.1823 RCIM_Cubic>,<at 0.2068 0.276>,<at 0.4703 0.2865>,<at 0.7289 0.4948>,<at 1 0.2396 RCIM_Cubic>]}
         RedCurve {keys [<at 0 0.1146 RCIM_Cubic>,<at 0.2068 0.174>,<at 0.4703 0.1785>,<at 0.7289 0.3322>,<at 1 0.1586 RCIM_Cubic>]}
        }})
     )
    )) {isSetParameters=true, module=SetVariables_#1}
   /SetVariables(Particles.MaterialRandom:NiagaraFloat=&RandomRangeFloat(Minimum=0, Maximum=1)) {isSetParameters=true, module=SetVariables_#2}
  ]
  particleUpdateScript [
   &ScaleVelocity(Velocity Scale=&MakeVector(
     X=&Multiply_Float(
      A=0.9
      B=&FloatFromCurve(
       FloatCurve=DataInterface() {propertyValues {Curve {keys [<at 0 0.9>,<at 1 3>]}}}
       CurveIndex=&Divide_Float(A=&VectorLength(Vector=$Engine.Owner.Velocity), B=9000)
      )
     )
     Y=&Multiply_Float(
      A=0.9
      B=&FloatFromCurve(
       FloatCurve=DataInterface() {propertyValues {Curve {keys [<at 0 0.9>,<at 1 3>]}}}
       CurveIndex=&Divide_Float(A=&VectorLength(Vector=$Engine.Owner.Velocity), B=9000)
      )
     )
     Z=&Multiply_Float(
      A=0.9
      B=&FloatFromCurve(
       FloatCurve=DataInterface() {propertyValues {Curve {keys [<at 0 0.9>,<at 1 3>]}}}
       CurveIndex=&Divide_Float(A=&VectorLength(Vector=$Engine.Owner.Velocity), B=9000)
      )
     )
    )) {module=ScaleVelocity001}
   &SolveForcesAndVelocity(Clamp Velocity=false, Limit Acceleration=false)
   /Game/Effects/Niagara/Modules/Ribbons/FN_UpdateRibbon_v6(Enable Skeletal Position=true, SkeletalMesh:NiagaraDataInterfaceSkeletalMesh=$User.SkeletalMesh)
   &DynamicMaterialParameters(
    Index 0 Param 1=&MaskFloatBySpawnGroup(Masked Float=0, Passthrough Float=0.5)
    Index 0 Param 2=&FloatFromCurve(
     FloatCurve=DataInterface() {propertyValues {Curve {keys [
         <at 0 0.3>
         {arriveTangent=-2.5521, interpMode=RCIM_Cubic, leaveTangent=-2.5521, tangentMode=RCTM_User, time=0.1, value=0.75}
         {arriveTangent=-0.0873, interpMode=RCIM_Cubic, leaveTangent=0.0167, tangentMode=RCTM_Break, time=1, value=0}
        ]}}}
     Scale Curve=&FN_FrontEndFloatModifier(
      Gameplay=&Lerp_Float(A=&Lerp_Float(A=0.3, B=0, Alpha=$System.Timeline001.Time), B=0.8, Alpha=$System.Timeline.Time)
      Front End=0.8
     )
    )
    Index 0 Param 3=1
    Index 0 Param 4=1
    Module.Param0WriteEnabled=true
    Module.Param1WriteEnabled=true
    Module.Param2WriteEnabled=true
    Module.Param3WriteEnabled=true
   )
   &ScaleRibbonWidth(Ribbon Width Scale=&FloatFromCurve(
     FloatCurve=DataInterface() {propertyValues {Curve {keys [
         {arriveTangent=0.3835, interpMode=RCIM_Cubic, leaveTangent=0.3835, tangentMode=RCTM_User, time=0, value=1.25}
         {arriveTangent=4.7926, interpMode=RCIM_Cubic, leaveTangent=1.8929, tangentMode=RCTM_Break, time=0.3, value=1.75}
         <at 0.5 2>
         <at 0.7 1>
        ]}}}
     Scale Curve=0.6
    ))
   &ScaleColor(
    Scale Alpha=&Multiply_Float(
     A=&FNTimeOfDayScalar(Night Scale=1.5, Morning Scale=2, Evening Scale=2)
     B=&FloatFromCurve(FloatCurve=DataInterface() {propertyValues {Curve {keys [<at 0 0.8>,<at 0.3 0.95>,<at 1 1>]}}})
    )
    ScaleA=true
    ScaleRGB=true
   )
   &Color(Color=&MultiplyLinearColor(
     Linear Color=$Particles.Color
     Scale Factor=&MakeVector4_FromLinearColor(LinearColor=&FNTimeOfDayColor(
       Night Color=<rgba 0.724 0.5202 0.6934 0.854>
       Morning Color=<rgba 1 1 1 1>
       Day Color=<rgba 1 1 1 1>
       Evening Color=<rgba 0.9688 0.9688 0.9688 1>
      ))
    ))
  ]
  rendererClasses [&NiagaraRibbonRendererProperties]
  renderers [{rendererClass=&NiagaraRibbonRendererProperties}]
 }
 LongerRibbon001 {
  emitterUpdateScript [
   &EmitterState()
   &FN_SpawnRibbon_v3(Number of Ribbons=2, Skeleton Mesh:NiagaraDataInterfaceSkeletalMesh=$User.SkeletalMesh, SpawnRate=16)
  ]
  particleSpawnScript [
   &FN_SetSpawnGroup(Number of Spawn Groups=2)
   &FN_InitializeRibbon_v5(
    Enable Skeletal Position=true
    Lifetime=2
    Offset Amount=15
    Ribbon Width=100
    SkeletalMesh:NiagaraDataInterfaceSkeletalMesh=$User.SkeletalMesh
   )
   &AddVelocity(Velocity=&MakeVector(
     X=&Multiply_Float(A=55, B=&Multiply_Float(A=$User.RightDot, B=-1))
     Y=&FN_FrontEndFloatModifier(Gameplay=250, Front End=-400)
     Z=0
    ))
   /SetVariables(Particles.Color=&MakeCustomLinearColorFromBool(
     Bool=&SetBoolByFloatComparison(A=$User.VariantID, B=1, Comparison Type=::NewEnumerator6)
     True Color=&ColorFromCurve(
      ColorCurve=DataInterface() {propertyValues {
        AlphaCurve {keys [<at 0 1 RCIM_Cubic>]}
        BlueCurve {keys [<at 0 0.3312 RCIM_Cubic>,<at 0.289 0.0905>,<at 0.4933 0.1815>,<at 0.6282 0.2689>,<at 1 0.4333 RCIM_Cubic>]}
        GreenCurve {keys [<at 0 0.4728 RCIM_Cubic>,<at 0.289 0.1379>,<at 0.4933 0.2511>,<at 0.6282 0.4669>,<at 1 0.5708 RCIM_Cubic>]}
        RedCurve {keys [<at 0 0.776 RCIM_Cubic>,<at 0.289 0.3646>,<at 0.4933 0.776>,<at 0.6282 0.9323>,<at 1 0.7917 RCIM_Cubic>]}
       }}
      CurveIndex=&RandomRangeFloat(Minimum=0, Maximum=1)
     )
     False Color=&MakeCustomLinearColorFromBool(
      Bool=&SetBoolByFloatComparison(A=$User.VariantID, B=2, Comparison Type=::NewEnumerator6)
      True Color=&ColorFromCurve(ColorCurve=DataInterface() {propertyValues {
         AlphaCurve {keys [<at 0 1 RCIM_Cubic>]}
         BlueCurve {keys [<at 0 0.4123 RCIM_Cubic>,<at 0.289 0.1899>,<at 0.4933 0.1815>,<at 0.6282 0.4759>,<at 1 0.4333 RCIM_Cubic>]}
         GreenCurve {keys [<at 0 0.528 RCIM_Cubic>,<at 0.289 0.2471>,<at 0.4933 0.2511>,<at 0.6282 0.6121>,<at 1 0.5708 RCIM_Cubic>]}
         RedCurve {keys [<at 0 0.776 RCIM_Cubic>,<at 0.289 0.5208>,<at 0.4933 0.776>,<at 0.6282 0.9323>,<at 1 0.7917 RCIM_Cubic>]}
        }})
      False Color=&ColorFromCurve(ColorCurve=DataInterface() {propertyValues {
         AlphaCurve {keys [<at 0 1 RCIM_Cubic>]}
         BlueCurve {keys [<at 0 0.305 RCIM_Cubic>,<at 0.289 0.0824>,<at 0.4933 0.1922>,<at 0.6282 0.4965>,<at 1 0.305 RCIM_Cubic>]}
         GreenCurve {keys [<at 0 0.7157 RCIM_Cubic>,<at 0.289 0.1875>,<at 0.4933 0.526>,<at 0.6282 0.9323>,<at 1 0.7157 RCIM_Cubic>]}
         RedCurve {keys [<at 0 0.2051 RCIM_Cubic>,<at 0.289 0.0561>,<at 0.4933 0.23>,<at 0.6282 0.4926>,<at 1 0.2051 RCIM_Cubic>]}
        }})
     )
    )) {isSetParameters=true, module=SetVariables_#1}
   /SetVariables(Particles.MaterialRandom:NiagaraFloat=&RandomRangeFloat(Minimum=0, Maximum=1)) {isSetParameters=true, module=SetVariables_#2}
  ]
  particleUpdateScript [
   &ScaleVelocity(Velocity Scale=&MakeVector(
     X=&Multiply_Float(
      A=0.9
      B=&FloatFromCurve(
       FloatCurve=DataInterface() {propertyValues {Curve {keys [<at 0 0.9>,<at 1 3>]}}}
       CurveIndex=&Divide_Float(A=&VectorLength(Vector=$Engine.Owner.Velocity), B=9000)
      )
     )
     Y=&Multiply_Float(
      A=0.9
      B=&FloatFromCurve(
       FloatCurve=DataInterface() {propertyValues {Curve {keys [<at 0 0.9>,<at 1 3>]}}}
       CurveIndex=&Divide_Float(A=&VectorLength(Vector=$Engine.Owner.Velocity), B=9000)
      )
     )
     Z=&Multiply_Float(
      A=0.9
      B=&FloatFromCurve(
       FloatCurve=DataInterface() {propertyValues {Curve {keys [<at 0 0.9>,<at 1 3>]}}}
       CurveIndex=&Divide_Float(A=&VectorLength(Vector=$Engine.Owner.Velocity), B=9000)
      )
     )
    )) {module=ScaleVelocity001}
   &SolveForcesAndVelocity(Clamp Velocity=false, Limit Acceleration=false)
   &FN_UpdateRibbon_v6(Enable Skeletal Position=true, SkeletalMesh:NiagaraDataInterfaceSkeletalMesh=$User.SkeletalMesh)
   &DynamicMaterialParameters(
    Index 0 Param 1=&MaskFloatBySpawnGroup(Masked Float=0.5, Passthrough Float=1)
    Index 0 Param 2=&FloatFromCurve(
     FloatCurve=DataInterface() {propertyValues {Curve {keys [
         <at 0 0.3>
         {arriveTangent=-2.8666, interpMode=RCIM_Cubic, leaveTangent=-2.8666, tangentMode=RCTM_User, time=0.1, value=0.75}
         {arriveTangent=-0.0873, interpMode=RCIM_Cubic, leaveTangent=0.0167, tangentMode=RCTM_Break, time=1, value=0}
        ]}}}
     Scale Curve=&FN_FrontEndFloatModifier(
      Gameplay=&Lerp_Float(A=&Lerp_Float(A=0.3, B=0, Alpha=$System.Timeline001.Time), B=0.6, Alpha=$System.Timeline.Time)
      Front End=0.6
     )
    )
    Index 0 Param 3=0.6
    Index 0 Param 4=1
    Module.Param0WriteEnabled=true
    Module.Param1WriteEnabled=true
    Module.Param2WriteEnabled=true
    Module.Param3WriteEnabled=true
   )
   &ScaleRibbonWidth(Ribbon Width Scale=&FloatFromCurve(
     FloatCurve=DataInterface() {propertyValues {Curve {keys [
         {arriveTangent=0.3835, interpMode=RCIM_Cubic, leaveTangent=0.3835, tangentMode=RCTM_User, time=0, value=0.85}
         {arriveTangent=4.7926, interpMode=RCIM_Cubic, leaveTangent=1.8929, tangentMode=RCTM_Break, time=0.3, value=1.75}
         <at 0.5 2>
         <at 0.7 1>
        ]}}}
     Scale Curve=0.6
    ))
   &ScaleColor(
    Scale Alpha=&Multiply_Float(
     A=&FNTimeOfDayScalar(Night Scale=1.5, Morning Scale=2, Evening Scale=2)
     B=&FloatFromCurve(FloatCurve=DataInterface() {propertyValues {Curve {keys [<at 0 0.8>,<at 0.3 0.95>,<at 1 1>]}}})
    )
    Scale RGB=&Multiply_Vector(
     A=&MakeVectorFromLinearColorRGB(Color=<rgba 0.901 0.804 0.6401 1>)
     B=&VectorFromFloat(Value=3.5)
    )
    ScaleA=true
    ScaleRGB=true
   )
   &Color(Color=&MultiplyLinearColor(
     Linear Color=$Particles.Color
     Scale Factor=&MakeVector4_FromLinearColor(LinearColor=&FNTimeOfDayColor(
       Night Color=<rgba 0.724 0.5202 0.6934 0.854>
       Morning Color=<rgba 1 1 1 1>
       Day Color=<rgba 1 1 1 1>
       Evening Color=<rgba 0.9688 0.9688 0.9688 1>
      ))
    ))
  ]
  rendererClasses [&NiagaraRibbonRendererProperties]
  renderers [{rendererClass=&NiagaraRibbonRendererProperties}]
 }
 SpeedLines {
  emitterUpdateScript [
   &EmitterState()
   &SpawnRate(
    SpawnRate=&MakeCustomFloatFromBool(Bool=$User.bIsFrontEndPreview, True Float=50, False Float=90)
    Use Spawn Probability=true
   ) {module=SpawnRate001}
  ]
  particleSpawnScript [
   &InitializeParticle(
    Color=&MakeLinearColorFromVectorAndFloat(
     "Vector (RGB)"=<1 1 1>
     "Float (Alpha)"=&Multiply_Float(
      A=&UniformRangedFloat(Minimum=0.15, Maximum=0.3)
      B=&MakeCustomFloatFromBool(Bool=$User.bIsFrontEndPreview, True Float=0.25, False Float=0.8)
     )
    )
    Lifetime Max=0.16
    Lifetime Min=0.1
    Lifetime Mode=::NewEnumerator1
    Sprite Size Max=<3 10>
    Sprite Size Min=<2 2>
    Sprite Size Mode=::NewEnumerator2
    UsePositionOffset=false
    Write Material Random=true
   )
   &TorusLocation(
    Align Torus With Axis=true
    Handle Radius=0
    Large Radius:NiagaraFloat=&FN_FrontEndFloatModifier(
     Gameplay=&RandomRangeFloat(Minimum=250, Maximum=300)
     Front End=&RandomRangeFloat(Minimum=200, Maximum=250)
    )
    MaskSpawnGroup=false
    Module.Override Randomness=false
    Offset=<0 -180 0>
    Torus Axis=<0 1 0>
    Torus Origin:NiagaraPosition=&Add_Vector(
     A=Linked() {linkedVariable {name=Particles.Position, type=NiagaraPosition}}
     B=&MakeVector(
      X=0
      Y=&MakeCustomFloatFromBool(Bool=$User.bIsFrontEndPreview, True Float=80, False Float=-100)
      Z=&MakeCustomFloatFromBool(Bool=$User.bIsFrontEndPreview, True Float=0, False Float=20)
     )
    )
   )
   &AddVelocityFromPoint(
    Module.Use Falloff Distance=false
    Origin Offset=&FN_FrontEndVector3Modifier(Gameplay=<0 -700 0>, Front End=<0 0 0>)
    Velocity Strength=&Multiply_Float(A=500, B=&MakeCustomFloatFromBool(Bool=$User.bIsFrontEndPreview, True Float=0, False Float=1))
   )
   &AddVelocity(Velocity=&Multiply_Vector(
     A=&FN_FrontEndVector3Modifier(Gameplay=&RandomRangeVector(Minimum=<0 300 0>, Maximum=<0 700 0>), Front End=<0 100 10>)
     B=&MakeVector(X=1, Y=&MakeCustomFloatFromBool(Bool=$User.bIsFrontEndPreview, True Float=1, False Float=1), Z=1)
    ))
  ]
  particleUpdateScript [
   &SolveForcesAndVelocity(Clamp Velocity=false, Limit Acceleration=false)
   &ParticleState()
   &ScaleSpriteSize(Scale Factor=&Vector2DFromCurve(
     Vector2Curve=DataInterface() {propertyValues {XCurve {keys [<at 0 0.5 RCIM_Cubic>,<at 0.5 1 RCIM_Cubic>]}, YCurve {keys [<at 0 0 RCIM_Cubic>,<at 0.5 1 RCIM_Cubic>]}}}
     ScaleCurve=&Vector2DFromFloat(Value=1)
    ))
   &ScaleSpriteSizeBySpeed(Max Scale Factor=<1 50>, Min Scale Factor=<0.01 0.05>, Velocity Threshold=1)
   &ScaleColor(
    Scale Alpha=&FloatFromCurve(
     FloatCurve=DataInterface() {propertyValues {Curve {keys [
         <at 0 0 RCIM_Cubic>
         <at 0.3759 1>
         {arriveTangent=-2.221, interpMode=RCIM_Cubic, leaveTangent=-2.221, tangentMode=RCTM_User, time=1, value=0}
        ]}}}
     Scale Curve=&MakeCustomFloatFromBool(
      Bool=$User.bIsFrontEndPreview
      True Float=&MakeCustomFloatFromBool(Bool=$User.bIsFullyDeployed, True Float=1, False Float=0)
      False Float=&Lerp_Float(A=0, B=1, Alpha=$User.ForwardDot)
     )
    )
    Scale RGB=&VectorFromFloat(Value=2)
    ScaleA=true
    ScaleRGB=true
   )
  ]
  rendererClasses [&NiagaraSpriteRendererProperties]
  renderers [{rendererClass=&NiagaraSpriteRendererProperties}]
 }
 TrailBlue {
  emitterSpawnScript [&FN_RibbonIDOffsetUtility()]
  emitterUpdateScript [
   &EmitterState(Life Cycle Mode=::NewEnumerator1, Loop Delay=$User.AnimationDelay, Loop Duration=4, UseLoopDelay=false)
   &FN_SpawnAnimTrail(AnimTrailWidthMode=::FromSecond, Number of Anim Trails=2)
  ]
  particleSpawnScript [
   &FN_InitializeAnimTrailBoneIndecies(First Bone or Socket Index=1, Second Bone or Socket Index=0)
   &FN_InitializeAnimTrailBoneIndecies(AnimTrail Selection=1, First Bone or Socket Index=3, Second Bone or Socket Index=2) {module=FN_InitializeAnimTrailBoneIndecies001}
   /SetVariables(Particles.AlphaFadeIn:NiagaraFloat=&Min_Float(A:NiagaraFloat=&Divide_Float(A=$Emitter.Age, B=1), B:NiagaraFloat=1)) {isSetParameters=true, module=SetVariables_#3}
   &FN_InitializeAnimTrail(
    AnimTrailWidth=2
    Color=&MakeLinearColor(
     R:NiagaraFloat=1
     G:NiagaraFloat=1
     B:NiagaraFloat=1
     A:NiagaraFloat=&Multiply_Float(A=$Particles.AlphaFadeIn, B=0.8)
    )
    Lifetime=&MakeCustomFloatFromBool(Bool=$Particles.Locked, True Float=100, False Float=1.5)
    Position Offset=<0 -20 0>
   )
   &AddVelocity(Velocity=&FN_FrontEndVector3Modifier(Gameplay=<0 300 0>, Front End=<0 -270 0>))
  ]
  particleUpdateScript [
   &SolveForcesAndVelocity(Clamp Velocity=false, Limit Acceleration=false)
   &FN_UpdateAnimTrail(
    Locked Particle Color:LinearColor=&MultiplyLinearColor(
     Linear Color=$Particles.Initial.Color
     Scale Factor=&MakeFloat4FromVectorAndFloat("Vector (RGB)"=<1 1 1>, "Float (Alpha)"=$System.FN_Integrator001.Value)
    )
    Locked Particle RibbonWidth=1
    SkeletalMesh=$User.SkeletalMesh
   ) {module=FN_UpdateAnimTrail001}
   /SetVariables(Particles.RibbonU0Override=$Particles.NormalizedAge) {isSetParameters=true, module=SetVariables_#4}
   &Color(
    Color=&MakeCustomLinearColorFromBool(
     Bool=&SetBoolByFloatComparison(A=$User.VariantID, B=1, Comparison Type=::NewEnumerator6)
     True Color=&SelectLinearColorFromArray(
      Array Sampling Mode=::NewEnumerator1
      Direct Array Index=$Particles.SpawnGroup
      Color Selection Array=DataInterface() {propertyValues {ColorData [<rgba 0.1054 1 0.5861 1>, <rgba 1 0.1094 0.7312 1>]}}
      Random Seed=?
     )
     False Color=&MakeCustomLinearColorFromBool(
      Bool=&SetBoolByFloatComparison(A=$User.VariantID, B=2, Comparison Type=::NewEnumerator6)
      True Color=&SelectLinearColorFromArray(
       Array Sampling Mode=::NewEnumerator1
       Direct Array Index=&SpawnGroup()
       Color Selection Array=DataInterface() {propertyValues {ColorData [<rgba 0.4635 0.1655 0.0817 3>, <rgba 1 0.7745 0.5677 1>]}}
       Random Seed=?
      )
      False Color=&SelectLinearColorFromArray(
       Array Sampling Mode=::NewEnumerator1
       Direct Array Index=&SpawnGroup()
       Color Selection Array=DataInterface() {propertyValues {ColorData [<rgba 0.3156 0.8125 0.1969 1>, <rgba 0.0997 0.1927 0.0647 2>]}}
       Random Seed=?
      )
     )
    )
    Scale Alpha=&Multiply_Float(
     A=&Multiply_Float(A=0.35, B=$Particles.AlphaFadeIn)
     B=&FN_FrontEndFloatModifier(Gameplay=$System.FN_Integrator001.Value, Front End=1)
    )
   )
  ]
  rendererClasses [&NiagaraRibbonRendererProperties]
  renderers [{rendererClass=&NiagaraRibbonRendererProperties}]
 }
 TrailIceCream {
  emitterSpawnScript [&FN_RibbonIDOffsetUtility()]
  emitterUpdateScript [
   &EmitterState(Life Cycle Mode=::NewEnumerator1, Loop Delay=$User.AnimationDelay, Loop Duration=1, UseLoopDelay=false)
   &FN_SpawnAnimTrail(Number of Anim Trails=2)
  ]
  particleSpawnScript [
   &FN_InitializeAnimTrailBoneIndecies(First Bone or Socket Index=5, Second Bone or Socket Index=4)
   &FN_InitializeAnimTrailBoneIndecies(AnimTrail Selection=1, First Bone or Socket Index=4, Second Bone or Socket Index=5) {module=FN_InitializeAnimTrailBoneIndecies001}
   /SetVariables(Particles.AlphaFadeIn:NiagaraFloat=&Min_Float(A:NiagaraFloat=&Divide_Float(A=$Emitter.Age, B=1), B:NiagaraFloat=1)) {isSetParameters=true, module=SetVariables_#3}
   &FN_InitializeAnimTrail(
    Color=&MakeLinearColor(
     R:NiagaraFloat=1
     G:NiagaraFloat=1
     B:NiagaraFloat=1
     A:NiagaraFloat=&Multiply_Float(A=$Particles.AlphaFadeIn, B=0.2)
    )
    Lifetime=&MakeCustomFloatFromBool(Bool=$Particles.Locked, True Float=100, False Float=1.5)
    Position Offset=<0 -20 0>
   )
   &AddVelocity(Velocity=&FN_FrontEndVector3Modifier(Gameplay=<0 300 0>, Front End=<0 -270 0>))
  ]
  particleUpdateScript [
   &SolveForcesAndVelocity(Clamp Velocity=false, Limit Acceleration=false)
   &FN_UpdateAnimTrail(
    Locked Particle Color:LinearColor=&MultiplyLinearColor(
     Linear Color=$Particles.Initial.Color
     Scale Factor=&MakeFloat4FromVectorAndFloat("Vector (RGB)"=<1 1 1>, "Float (Alpha)"=$System.FN_Integrator001.Value)
    )
    Locked Particle RibbonWidth=1
    SkeletalMesh=$User.SkeletalMesh
   ) {module=FN_UpdateAnimTrail001}
   /SetVariables(Particles.RibbonU0Override=$Particles.NormalizedAge) {isSetParameters=true, module=SetVariables_#4}
   &Color(
    Color=&SelectLinearColorFromArray(
     Array Sampling Mode=::NewEnumerator1
     Direct Array Index=&MakeIntFromFloat(Float=$User.VariantID)
     Color Selection Array=DataInterface() {propertyValues {ColorData [<rgba 0.8714 0.6172 0.4342 0.8>, <rgba 0.6094 0.2779 0.0825 1>, <rgba 0.7456 0.9219 0.4775 0.7>]}}
     Random Seed=?
    )
    Scale Alpha=&Multiply_Float(
     A=&Multiply_Float(A=0.35, B=$Particles.AlphaFadeIn)
     B=&FN_FrontEndFloatModifier(Gameplay=$System.FN_Integrator001.Value, Front End=1)
    )
   )
  ]
  rendererClasses [&NiagaraRibbonRendererProperties]
  renderers [{rendererClass=&NiagaraRibbonRendererProperties}]
 }
}
summary {
 systemName=NS_Glider_LilSplit
 emitters [
  {emitterName=TrailBlue, rendererClasses [&NiagaraRibbonRendererProperties]}
  {emitterName=TrailIceCream, rendererClasses [&NiagaraRibbonRendererProperties]}
  {emitterName=SpeedLines, rendererClasses [&NiagaraSpriteRendererProperties]}
  {emitterName=Empty, rendererClasses [&NiagaraSpriteRendererProperties]}
  {
   bEnabled=false
   emitterName=Empty001
   rendererClasses [&NiagaraSpriteRendererProperties, &NiagaraMeshRendererProperties]
  }
  {emitterName=Drops, simTarget=GPUComputeSim, rendererClasses [&NiagaraSpriteRendererProperties]}
  {emitterName=FN_AttachedRibbon, rendererClasses [&NiagaraRibbonRendererProperties]}
  {emitterName=LongerRibbon, rendererClasses [&NiagaraRibbonRendererProperties]}
  {emitterName=Drops001, simTarget=GPUComputeSim, rendererClasses [&NiagaraSpriteRendererProperties]}
  {emitterName=LongerRibbon001, rendererClasses [&NiagaraRibbonRendererProperties]}
 ]
 userVariables [
  User.AnimationDelay {type=NiagaraFloat, defaultValue=2, description=""}
  User.bIsFrontEnd {type=NiagaraBool}
  User.bIsFrontEndPreview {type=NiagaraBool}
  User.bIsFullyDeployed {type=NiagaraBool, defaultValue=true}
  User.ForwardDot {type=NiagaraFloat, defaultValue=1, description=""}
  User.RibbonIDOffset {type=NiagaraInt32}
  User.RightDot {type=NiagaraFloat, defaultValue=0, description=""}
  User.SkeletalMesh {
   type=NiagaraDataInterfaceSkeletalMesh
   defaultValue {
    struct=NiagaraExt_VariableValue_DataInterface
    value {dataInterface=&self:NiagaraDataInterfaceSkeletalMesh_0}
   }
  }
  User.VariantID {type=NiagaraFloat, defaultValue=1, description=""}
 ]
}
