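# Niagara deploy effect for the LilSplit glider, throwing two splashes of colored bits from the glider's skeletal mesh.
# Mirrored emitter sets (sprite/mesh chunks, palette-colored flecks, splash mesh, splash sprites) burst immediately at one bone/socket, then at a second after SecondSplashDelay.
# A separate delayed emitter briefly sprays falling, swirling particles from two spawn-group-indexed sockets.
# VariantID picks fleck palettes, system flavor variables color other pieces, and splashes adapt to time of day.
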NiagaraSystem /BRCosmetics/Gadgets/Assets/VinderTech_GliderChute/Glider_LilSplit/FX/NS_Glider_LilSplit_Deploy
dependencies {
 usedDataInterfaces [
  &NiagaraDataInterfaceCurve
  &NiagaraDataInterfaceArrayColor
  &NiagaraDataInterfaceVectorCurve
  &NiagaraDataInterfaceVector2DCurve
  &NiagaraDataInterfaceColorCurve
 ]
 usedDynamicInputs [&MakeIntFromFloat]
 usedModules [
  &SystemState
  &self:SystemSpawnScript.SystemScriptSource.SystemScriptGraph.NiagaraNodeAssignment_2.SetVariables_#1
  &self:SystemSpawnScript.SystemScriptSource.SystemScriptGraph.NiagaraNodeAssignment_1.SetVariables_#2
 ]
}
emitters {
 Empty {
  emitterUpdateScript [
   &EmitterState(Life Cycle Mode=::NewEnumerator1, Loop Behavior=::NewEnumerator1, Loop Delay=$System.SecondSplashDelay, Loop Duration=1, UseLoopDelay=true)
   &SpawnBurst_Instantaneous(Spawn Count=8, Use Spawn Probability=false, UseLoopCountLimit=false)
  ]
  particleSpawnScript [
   &InitializeParticle(
    Color=$System.FlavColB
    Lifetime Max=1
    Lifetime Min=0.5
    Lifetime Mode=::NewEnumerator1
    Mesh Scale Mode=::NewEnumerator1
    Mesh Uniform Scale Max=5
    Mesh Uniform Scale Min=2
    Position Offset=<0 0 -15>
    Sprite Rotation Mode=::NewEnumerator2
    Sprite Size Mode=::NewEnumerator1
    Uniform Sprite Size Max=30
    Uniform Sprite Size Min=15
    UsePositionOffset=false
    Write Material Random=true
   )
   &ShapeLocation(MaskSpawnGroup=false, Override Seed=false, Sphere Radius=35, UseOwnerScale=true)
   &SkeletalMeshLocation(
    Bone / Socket Index=1
    Bone Sampling Mode=::NewEnumerator4
    MaskSpawnGroup=false
    Sampled Position Offset=$User.ParticleOffsetRight
    Skeletal Mesh=$User.SkeletalMesh
   )
   &AddVelocity(Velocity=&RandomRangeVector(Minimum=<120 120 250>, Maximum=<-120 -120 500>))
  ]
  particleUpdateScript [
   &ParticleState()
   &ScaleMeshSize(Scale Factor=&VectorFromFloat(Value=&FloatFromCurve(FloatCurve=DataInterface() {propertyValues {Curve {keys [
          {tangentMode=RCTM_User, time=0, value=0.35}
          {interpMode=RCIM_Cubic, leaveTangent=0.2166, tangentMode=RCTM_Break, time=0.3, value=0.95}
          {arriveTangent=-2.3045, interpMode=RCIM_Cubic, tangentMode=RCTM_Break, time=1, value=0.35}
         ]}}})))
   &SpriteRotationRate(Rotation Rate=&RandomRangeFloat(Minimum=-100, Maximum=100))
   &ScaleSpriteSize(Uniform Curve Sprite Scale=DataInterface() {propertyValues {Curve {keys [
        {tangentMode=RCTM_User, time=0, value=0.35}
        {interpMode=RCIM_Cubic, leaveTangent=0.2166, tangentMode=RCTM_Break, time=0.3, value=0.95}
        {arriveTangent=-2.3045, interpMode=RCIM_Cubic, tangentMode=RCTM_Break, time=1, value=0.35}
       ]}}})
   &GravityForce(Gravity=&RandomRangeVector(Minimum=<0 0 -1000>, Maximum=<0 0 -1600>))
   &MeshRotationForce(Rotation=<2 2 2>)
   &SolveForcesAndVelocity(Clamp Velocity=false, Limit Acceleration=false)
   &ScaleSpriteSizeBySpeed(Max Scale Factor=<2 4>, Min Scale Factor=<0.8 1.2>)
  ]
  rendererClasses [&NiagaraSpriteRendererProperties, &NiagaraMeshRendererProperties]
  renderers [
   {rendererClass=&NiagaraSpriteRendererProperties}
   {rendererClass=&NiagaraMeshRendererProperties, rendererIndex=1}
  ]
 }
 Empty001 {
  emitterUpdateScript [
   &EmitterState(Life Cycle Mode=::NewEnumerator1, Loop Behavior=::NewEnumerator1, Loop Delay=$System.SecondSplashDelay, Loop Duration=1, UseLoopDelay=true)
   &SpawnBurst_Instantaneous(Spawn Count=8, Use Spawn Probability=false, UseLoopCountLimit=false)
  ]
  particleSpawnScript [
   &InitializeParticle(
    Color=&MakeCustomLinearColorFromBool(
     Bool=&SetBoolByFloatComparison(A=$User.VariantID, B=1, Comparison Type=::NewEnumerator6)
     True Color=&SelectLinearColorFromArray(
      Array Sampling Mode=::NewEnumerator0
      Direct Array Index=?
      Color Selection Array=DataInterface() {propertyValues {ColorData [<rgba 0.8958 0.3028 0 1>, <rgba 1 0.1094 0.7312 1>, <rgba 0.2135 0.7811 1 1>, <rgba 0.4838 0.2276 1 0.995>]}}
      Random Seed=0
     )
     False Color=&MakeCustomLinearColorFromBool(
      Bool=&SetBoolByFloatComparison(A=$User.VariantID, B=2, Comparison Type=::NewEnumerator6)
      True Color=&SelectLinearColorFromArray(
       Array Sampling Mode=::NewEnumerator0
       Direct Array Index=?
       Color Selection Array=DataInterface() {propertyValues {ColorData [<rgba 0.0573 0.016 0.0097 1>, <rgba 0.2656 0.0498 0.0524 1>, <rgba 0.1615 0.0438 0.0048 1>, <rgba 0.0417 0.0058 0 0.995>]}}
       Random Seed=0
      )
      False Color=&SelectLinearColorFromArray(
       Array Sampling Mode=::NewEnumerator0
       Direct Array Index=?
       Color Selection Array=DataInterface() {propertyValues {ColorData [<rgba 0.4677 0.6979 0.3344 1>, <rgba 0.1879 0.3333 0.0623 1>, <rgba 0.0647 0.0833 0.0126 1>, <rgba 0.3739 0.4115 0.0755 0.995>]}}
       Random Seed=0
      )
     )
    )
    Lifetime Max=0.8
    Lifetime Min=0.4
    Lifetime Mode=::NewEnumerator1
    Position Offset=<0 0 -15>
    Sprite Rotation Mode=::NewEnumerator2
    Sprite Size Max=<8 25>
    Sprite Size Min=<5 15>
    Sprite Size Mode=::NewEnumerator2
    UsePositionOffset=true
    Write Material Random=true
   )
   &ShapeLocation(MaskSpawnGroup=false, Override Seed=false, Sphere Radius=35, UseOwnerScale=true) {module=ShapeLocation001}
   &SkeletalMeshLocation(
    Bone / Socket Index=1
    Bone Sampling Mode=::NewEnumerator4
    MaskSpawnGroup=false
    Sampled Position Offset=$User.ParticleOffsetRight
    Skeletal Mesh=$User.SkeletalMesh
   )
   &ShapeLocation(MaskSpawnGroup=false, Override Seed=false, Sphere Radius=20, UseOwnerScale=true) {enabled=false}
   &AddVelocity(Velocity=&RandomRangeVector(Minimum=<120 120 250>, Maximum=<-120 -120 550>))
  ]
  particleUpdateScript [
   &ParticleState()
   &SpriteRotationRate(Rotation Rate=&RandomRangeFloat(Minimum=-250, Maximum=250))
   &ScaleSpriteSize(Uniform Curve Sprite Scale=DataInterface() {propertyValues {Curve {keys [
        {tangentMode=RCTM_User, time=0, value=0.35}
        {interpMode=RCIM_Cubic, leaveTangent=0.2166, tangentMode=RCTM_Break, time=0.3, value=0.95}
        {arriveTangent=-2.3045, interpMode=RCIM_Cubic, tangentMode=RCTM_Break, time=1, value=0.35}
       ]}}})
   &GravityForce(Gravity=&RandomRangeVector(Minimum=<0 0 -1000>, Maximum=<0 0 -1600>))
   &SolveForcesAndVelocity(Clamp Velocity=false, Limit Acceleration=false)
  ]
  rendererClasses [&NiagaraSpriteRendererProperties]
  renderers [{rendererClass=&NiagaraSpriteRendererProperties}]
 }
 Empty002 {
  emitterUpdateScript [
   &EmitterState(Life Cycle Mode=::NewEnumerator1, Loop Behavior=::NewEnumerator1, Loop Delay=$System.SecondSplashDelay, Loop Duration=1, UseLoopDelay=true)
   &SpawnBurst_Instantaneous(Use Spawn Probability=false, UseLoopCountLimit=false)
  ]
  particleSpawnScript [
   &InitializeParticle(
    Color=$System.FlavColB
    Lifetime Max=0.6
    Lifetime Min=0.3
    Lifetime Mode=::NewEnumerator1
    Mesh Scale Max=<5 5 5>
    Mesh Scale Min=<5 5 4>
    Mesh Scale Mode=::NewEnumerator2
    Position Offset=<0 0 -10>
    UsePositionOffset=true
    Write Material Random=true
   )
   &SkeletalMeshLocation(
    Bone / Socket Index=1
    Bone Sampling Mode=::NewEnumerator4
    MaskSpawnGroup=false
    Sampled Position Offset=$User.ParticleOffsetRight
    Skeletal Mesh=$User.SkeletalMesh
   )
   &InitialMeshOrientation(
    Mesh Orientation Mode=::NewEnumerator2
    Rotation=&RandomRangeVector(Minimum=<0 0 -1>, Maximum=<0 0 1>)
    Rotation Coordinate Space=::NewEnumerator0
    Use Rotation Vector=true
   )
   &AddVelocity(Velocity=<0 0 250>)
   &Color(Color=&MultiplyLinearColor(
     Linear Color=$Particles.Color
     Scale Factor=&MakeVector4_FromLinearColor(LinearColor=&FN_Select_Color_from_Time_Of_Day(Morning=<rgba 1 1 1 1>, Day=<rgba 1 1 1 1>, Evening=<rgba 1 0.664 0.6341 1>, Night=<rgba 0.7226 0.7425 1 1>))
    ))
  ]
  particleUpdateScript [
   &ParticleState()
   &ScaleMeshSize(Scale Factor=&VectorFromCurve(VectorCurve=DataInterface() {propertyValues {
       XCurve {keys [<at 0 0.25>,<at 0.3 0.625>,<at 1 1>]}
       YCurve {keys [<at 0 0.25>,<at 0.3 0.625>,<at 1 1>]}
       ZCurve {keys [
         <at 0 0.45>
         {interpMode=RCIM_Cubic, leaveTangent=-0.0185, tangentMode=RCTM_Break, time=0.2, value=1}
         {arriveTangent=-0.9743, interpMode=RCIM_Cubic, leaveTangent=-2.4944, tangentMode=RCTM_Break, time=0.7, value=0.8}
        ]}
      }}))
   &ScaleColor(Scale Alpha=1.5, ScaleA=true, ScaleRGB=true)
   &DynamicMaterialParameters(
    Index 0 Param 1=&Lerp_Float(
     A=&RandomRangeFloat(Minimum=0.2, Maximum=-0.2)
     B=-1
     Alpha=&FloatFromCurve(FloatCurve=DataInterface() {propertyValues {Curve {keys [
          {interpMode=RCIM_Cubic, tangentMode=RCTM_Break, time=0, value=0}
          {arriveTangent=1.5716, interpMode=RCIM_Cubic, tangentMode=RCTM_Break, time=1, value=1}
         ]}}})
    )
    Module.Index0Param0Write=true
    Module.Index0Param1Write=true
    Module.Index0Param2Write=true
    Module.Index0Param3Write=true
   )
   &GravityForce(Gravity=<0 0 -800>)
   &SolveForcesAndVelocity(Clamp Velocity=false, Limit Acceleration=false)
  ]
  rendererClasses [&NiagaraMeshRendererProperties]
  renderers [{rendererClass=&NiagaraMeshRendererProperties, rendererIndex=0}]
 }
 Empty003 {
  emitterUpdateScript [
   &EmitterState()
   &SpawnBurst_Instantaneous(Spawn Count=8, Use Spawn Probability=false, UseLoopCountLimit=false)
  ]
  particleSpawnScript [
   &InitializeParticle(
    Color=$System.FlavColB
    Lifetime Max=1
    Lifetime Min=0.5
    Lifetime Mode=::NewEnumerator1
    Mesh Scale Mode=::NewEnumerator1
    Mesh Uniform Scale Max=5
    Mesh Uniform Scale Min=2
    Position Offset=<0 0 20>
    Sprite Rotation Mode=::NewEnumerator2
    Sprite Size Mode=::NewEnumerator1
    Uniform Sprite Size Max=20
    Uniform Sprite Size Min=8
    UsePositionOffset=false
    Write Material Random=true
   )
   &ShapeLocation(MaskSpawnGroup=false, Override Seed=false, Sphere Radius=35, UseOwnerScale=true)
   &SkeletalMeshLocation(Bone Sampling Mode=::NewEnumerator4, MaskSpawnGroup=false, Sampled Position Offset=$User.ParticleOffset, Skeletal Mesh=$User.SkeletalMesh)
   &AddVelocity(Velocity=&RandomRangeVector(Minimum=<120 120 250>, Maximum=<-120 -120 500>))
  ]
  particleUpdateScript [
   &ParticleState()
   &ScaleMeshSize(Scale Factor=&VectorFromFloat(Value=&FloatFromCurve(FloatCurve=DataInterface() {propertyValues {Curve {keys [
          {tangentMode=RCTM_User, time=0, value=0.35}
          {interpMode=RCIM_Cubic, leaveTangent=0.2166, tangentMode=RCTM_Break, time=0.3, value=0.95}
          {arriveTangent=-2.3045, interpMode=RCIM_Cubic, tangentMode=RCTM_Break, time=1, value=0.35}
         ]}}})))
   &SpriteRotationRate(Rotation Rate=&RandomRangeFloat(Minimum=-100, Maximum=100))
   &ScaleSpriteSize(Uniform Curve Sprite Scale=DataInterface() {propertyValues {Curve {keys [
        {tangentMode=RCTM_User, time=0, value=0.35}
        {interpMode=RCIM_Cubic, leaveTangent=0.2166, tangentMode=RCTM_Break, time=0.3, value=0.95}
        {arriveTangent=-2.3045, interpMode=RCIM_Cubic, tangentMode=RCTM_Break, time=1, value=0.35}
       ]}}})
   &GravityForce(Gravity=&RandomRangeVector(Minimum=<0 0 -1000>, Maximum=<0 0 -1600>))
   &MeshRotationForce(Rotation=<2 2 2>)
   &SolveForcesAndVelocity(Clamp Velocity=false, Limit Acceleration=false)
   &ScaleSpriteSizeBySpeed(Max Scale Factor=<2 4>, Min Scale Factor=<0.8 1.2>)
  ]
  rendererClasses [&NiagaraSpriteRendererProperties, &NiagaraMeshRendererProperties]
  renderers [
   {rendererClass=&NiagaraSpriteRendererProperties}
   {rendererClass=&NiagaraMeshRendererProperties, rendererIndex=1}
  ]
 }
 Empty004 {
  emitterUpdateScript [
   &EmitterState(Life Cycle Mode=::NewEnumerator1, Loop Behavior=::NewEnumerator1, Loop Delay=1.8, Loop Duration=0.3, UseLoopDelay=true)
   &SpawnBurst_Instantaneous(Spawn Count=4, Use Spawn Probability=false, UseLoopCountLimit=false)
   &SpawnRate(SpawnRate=70, Use Spawn Probability=false)
  ]
  particleSpawnScript [
   &FN_SetSpawnGroup(Number of Spawn Groups=2)
   &InitializeParticle(
    Color=$System.FlavColA
    Lifetime Max=0.7
    Lifetime Min=0.5
    Lifetime Mode=::NewEnumerator1
    Mesh Scale Mode=::NewEnumerator1
    Mesh Uniform Scale Max=5
    Mesh Uniform Scale Min=2
    Position Offset=<0 0 20>
    Sprite Rotation Mode=::NewEnumerator2
    Sprite Size Mode=::NewEnumerator1
    Uniform Sprite Size Max=30
    Uniform Sprite Size Min=15
    UsePositionOffset=false
    Write Material Random=true
   )
   &SkeletalMeshLocation(
    Bone / Socket Index=$Particles.SpawnGroup
    Bone Sampling Mode=::NewEnumerator4
    MaskSpawnGroup=false
    Sampled Position Offset=<0 0 30>
    Sampled Position Offset Coordinate Space:ENiagaraSkelMeshTransforms=::NewEnumerator0
    Skeletal Mesh=$User.SkeletalMesh
   )
   &ShapeLocation(MaskSpawnGroup=false, Override Seed=false, Sphere Radius=20, UseOwnerScale=true)
  ]
  particleUpdateScript [
   &ParticleState()
   &ScaleMeshSize(Scale Factor=&VectorFromFloat(Value=&FloatFromCurve(FloatCurve=DataInterface() {propertyValues {Curve {keys [
          {tangentMode=RCTM_User, time=0, value=0.35}
          {interpMode=RCIM_Cubic, leaveTangent=0.2166, tangentMode=RCTM_Break, time=0.3, value=0.95}
          {arriveTangent=-2.3045, interpMode=RCIM_Cubic, tangentMode=RCTM_Break, time=1, value=0.35}
         ]}}})))
   &SpriteRotationRate(Rotation Rate=&RandomRangeFloat(Minimum=-100, Maximum=100))
   &ScaleSpriteSize(Uniform Curve Sprite Scale=DataInterface() {propertyValues {Curve {keys [
        {tangentMode=RCTM_User, time=0, value=0.725}
        {interpMode=RCIM_Cubic, leaveTangent=0.2166, tangentMode=RCTM_Break, time=0.3, value=0.95}
        {arriveTangent=-1.9059, interpMode=RCIM_Cubic, tangentMode=RCTM_Break, time=1, value=0.4}
       ]}}})
   &VortexForce(UseOriginOffset=false)
   &GravityForce(Gravity=&RandomRangeVector(Minimum=<0 0 -1450>, Maximum=<0 0 -2000>))
   &MeshRotationForce(Rotation=<2 2 2>)
   &SolveForcesAndVelocity(Clamp Velocity=false, Limit Acceleration=false)
  ]
  rendererClasses [&NiagaraSpriteRendererProperties, &NiagaraMeshRendererProperties]
  renderers [
   {rendererClass=&NiagaraSpriteRendererProperties}
   {rendererClass=&NiagaraMeshRendererProperties, rendererIndex=1}
  ]
 }
 Empty005 {
  emitterUpdateScript [
   &EmitterState()
   &SpawnBurst_Instantaneous(Spawn Count=8, Use Spawn Probability=false, UseLoopCountLimit=false)
  ]
  particleSpawnScript [
   &InitializeParticle(
    Color=&MakeCustomLinearColorFromBool(
     Bool=&SetBoolByFloatComparison(A=$User.VariantID, B=1, Comparison Type=::NewEnumerator6)
     True Color=&SelectLinearColorFromArray(
      Array Sampling Mode=::NewEnumerator0
      Direct Array Index=?
      Color Selection Array=DataInterface() {propertyValues {ColorData [<rgba 0.8958 0.3028 0 1>, <rgba 1 0.1094 0.7312 1>, <rgba 0.2135 0.7811 1 1>, <rgba 0.4838 0.2276 1 0.995>]}}
      Random Seed=0
     )
     False Color=&MakeCustomLinearColorFromBool(
      Bool=&SetBoolByFloatComparison(A=$User.VariantID, B=2, Comparison Type=::NewEnumerator6)
      True Color=&SelectLinearColorFromArray(
       Array Sampling Mode=::NewEnumerator0
       Direct Array Index=?
       Color Selection Array=DataInterface() {propertyValues {ColorData [<rgba 0.0573 0.016 0.0097 1>, <rgba 0.2656 0.0498 0.0524 1>, <rgba 0.1615 0.0438 0.0048 1>, <rgba 0.0417 0.0058 0 0.995>]}}
       Random Seed=0
      )
      False Color=&SelectLinearColorFromArray(
       Array Sampling Mode=::NewEnumerator0
       Direct Array Index=?
       Color Selection Array=DataInterface() {propertyValues {ColorData [<rgba 0.4677 0.6979 0.3344 1>, <rgba 0.1879 0.3333 0.0623 1>, <rgba 0.0647 0.0833 0.0126 1>, <rgba 0.3739 0.4115 0.0755 0.995>]}}
       Random Seed=0
      )
     )
    )
    Lifetime Max=0.8
    Lifetime Min=0.4
    Lifetime Mode=::NewEnumerator1
    Position Offset=<0 0 20>
    Sprite Rotation Mode=::NewEnumerator2
    Sprite Size Max=<8 25>
    Sprite Size Min=<5 15>
    Sprite Size Mode=::NewEnumerator2
    UsePositionOffset=true
    Write Material Random=true
   )
   &ShapeLocation(MaskSpawnGroup=false, Override Seed=false, Sphere Radius=35, UseOwnerScale=true) {module=ShapeLocation001}
   &SkeletalMeshLocation(Bone Sampling Mode=::NewEnumerator4, MaskSpawnGroup=false, Sampled Position Offset=$User.ParticleOffset, Skeletal Mesh=$User.SkeletalMesh)
   &ShapeLocation(MaskSpawnGroup=false, Override Seed=false, Sphere Radius=20, UseOwnerScale=true) {enabled=false}
   &AddVelocity(Velocity=&RandomRangeVector(Minimum=<120 120 250>, Maximum=<-120 -120 550>))
  ]
  particleUpdateScript [
   &ParticleState()
   &SpriteRotationRate(Rotation Rate=&RandomRangeFloat(Minimum=-250, Maximum=250))
   &ScaleSpriteSize(Uniform Curve Sprite Scale=DataInterface() {propertyValues {Curve {keys [
        {tangentMode=RCTM_User, time=0, value=0.35}
        {interpMode=RCIM_Cubic, leaveTangent=0.2166, tangentMode=RCTM_Break, time=0.3, value=0.95}
        {arriveTangent=-2.3045, interpMode=RCIM_Cubic, tangentMode=RCTM_Break, time=1, value=0.35}
       ]}}})
   &GravityForce(Gravity=&RandomRangeVector(Minimum=<0 0 -1000>, Maximum=<0 0 -1600>))
   &SolveForcesAndVelocity(Clamp Velocity=false, Limit Acceleration=false)
  ]
  rendererClasses [&NiagaraSpriteRendererProperties]
  renderers [{rendererClass=&NiagaraSpriteRendererProperties}]
 }
 Empty006 {
  emitterUpdateScript [
   &EmitterState()
   &SpawnBurst_Instantaneous(Use Spawn Probability=false, UseLoopCountLimit=false)
  ]
  particleSpawnScript [
   &InitializeParticle(
    Color=$System.FlavColB
    Lifetime Max=0.6
    Lifetime Min=0.3
    Lifetime Mode=::NewEnumerator1
    Mesh Scale Max=<5 5 5>
    Mesh Scale Min=<5 5 4>
    Mesh Scale Mode=::NewEnumerator2
    Position Offset=<0 0 22>
    UsePositionOffset=true
    Write Material Random=true
   )
   &SkeletalMeshLocation(Bone Sampling Mode=::NewEnumerator4, MaskSpawnGroup=false, Sampled Position Offset=$User.ParticleOffset, Skeletal Mesh=$User.SkeletalMesh)
   &InitialMeshOrientation(
    Mesh Orientation Mode=::NewEnumerator2
    Rotation=&RandomRangeVector(Minimum=<0 0 -1>, Maximum=<0 0 1>)
    Rotation Coordinate Space=::NewEnumerator0
    Use Rotation Vector=true
   )
   &AddVelocity(Velocity=<0 0 220>)
   &Color(Color=&MultiplyLinearColor(
     Linear Color=$Particles.Color
     Scale Factor=&MakeVector4_FromLinearColor(LinearColor=&FN_Select_Color_from_Time_Of_Day(Morning=<rgba 1 1 1 1>, Day=<rgba 1 1 1 1>, Evening=<rgba 1 0.664 0.6341 1>, Night=<rgba 0.7226 0.7425 1 1>))
    ))
  ]
  particleUpdateScript [
   &ParticleState()
   &ScaleMeshSize(Scale Factor=&VectorFromCurve(VectorCurve=DataInterface() {propertyValues {
       XCurve {keys [<at 0 0.25>,<at 0.3 0.625>,<at 1 1>]}
       YCurve {keys [<at 0 0.25>,<at 0.3 0.625>,<at 1 1>]}
       ZCurve {keys [
         <at 0 0.45>
         {interpMode=RCIM_Cubic, leaveTangent=-0.0185, tangentMode=RCTM_Break, time=0.2, value=1}
         {arriveTangent=-0.9743, interpMode=RCIM_Cubic, leaveTangent=-2.4944, tangentMode=RCTM_Break, time=0.7, value=0.8}
        ]}
      }}))
   &ScaleColor(Scale Alpha=1.5, ScaleA=true, ScaleRGB=true)
   &DynamicMaterialParameters(
    Index 0 Param 1=&Lerp_Float(
     A=&RandomRangeFloat(Minimum=0.2, Maximum=-0.2)
     B=-1
     Alpha=&FloatFromCurve(FloatCurve=DataInterface() {propertyValues {Curve {keys [
          {interpMode=RCIM_Cubic, tangentMode=RCTM_Break, time=0, value=0}
          {arriveTangent=1.5716, interpMode=RCIM_Cubic, tangentMode=RCTM_Break, time=1, value=1}
         ]}}})
    )
    Module.Index0Param0Write=true
    Module.Index0Param1Write=true
    Module.Index0Param2Write=true
    Module.Index0Param3Write=true
   )
   &GravityForce(Gravity=<0 0 -800>)
   &SolveForcesAndVelocity(Clamp Velocity=false, Limit Acceleration=false)
  ]
  rendererClasses [&NiagaraMeshRendererProperties]
  renderers [{rendererClass=&NiagaraMeshRendererProperties, rendererIndex=0}]
 }
 Splash {
  emitterUpdateScript [
   &EmitterState(Life Cycle Mode=::NewEnumerator1, Loop Behavior=::NewEnumerator1, Loop Delay=$System.SecondSplashDelay, Loop Duration=1, UseLoopDelay=true)
   &SpawnBurst_Instantaneous(Use Spawn Probability=false, UseLoopCountLimit=false)
  ]
  particleSpawnScript [
   &InitializeParticle(
    Lifetime Max=0.45
    Lifetime Min=0.3
    Lifetime Mode=::NewEnumerator1
    Position Offset=<0 0 -10>
    Sprite Size Mode=::NewEnumerator3
    Uniform Sprite Size=&FN_FrontEndFloatModifier(Gameplay=110, Front End=75)
    UsePositionOffset=false
    Write Material Random=true
   )
   &SkeletalMeshLocation(
    Bone / Socket Index=1
    Bone Sampling Mode=::NewEnumerator4
    MaskSpawnGroup=false
    Sampled Position Offset=$User.ParticleOffsetRight
    Skeletal Mesh=$User.SkeletalMesh
   )
   &Color(Color=&MultiplyLinearColor(
     Linear Color=$Particles.Initial.Color
     Scale Factor=&MakeVector4_FromLinearColor(LinearColor=&FN_Select_Color_from_Time_Of_Day(Morning=<rgba 1 1 1 1>, Day=<rgba 1 1 1 1>, Evening=<rgba 1 0.7676 0.7377 1>, Night=<rgba 0.9141 0.8046 1.25 1>))
    )) {enabled=false}
   /SetVariables(
    Particles.Render2COlor:LinearColor=<rgba 0.9219 0.9057 0.8979 1>
    Particles.Render2Rotation:NiagaraFloat=&RandomRangeFloat(Minimum=-360, Maximum=360)
    Particles.Render2SpriteSize:Vector2f=&Vector2DFromFloat(Value=&FN_FrontEndFloatModifier(Gameplay=140, Front End=125))
   ) {isSetParameters=true, module=SetVariables_#3}
   &Color(Color=&MultiplyLinearColor(
     Linear Color=$Particles.Render2COlor
     Scale Factor=&MakeVector4_FromLinearColor(LinearColor=&FN_Select_Color_from_Time_Of_Day(Morning=<rgba 1 1 1 1>, Day=<rgba 1 1 1 1>, Evening=<rgba 1 0.7003 0.6404 1>, Night=<rgba 0.6748 0.5503 1 1>))
    )) {module=Color001}
  ]
  particleUpdateScript [
   &ParticleState()
   &ScaleSpriteSize(Uniform Curve Sprite Scale=DataInterface() {propertyValues {Curve {keys [
        {tangentMode=RCTM_User, time=0, value=0.6}
        {interpMode=RCIM_Cubic, leaveTangent=0.5568, tangentMode=RCTM_Break, time=0.2, value=0.9}
        {tangentMode=RCTM_User, time=1, value=1}
       ]}}})
   &ScaleColor(
    Linear Color Curve=DataInterface() {propertyValues {
      AlphaCurve {keys [<at 0.4558 1>,<at 1 0>]}
      BlueCurve {keys [<at 0.3172 1>]}
      GreenCurve {keys [<at 0.3172 1>]}
      RedCurve {keys [<at 0.3172 1>]}
     }}
    Scale Mode=::NewEnumerator2
   )
   &GravityForce(Coordinate Space=::Simulation, Gravity=<0 0 -800>)
   &SolveForcesAndVelocity(Clamp Velocity=false, Limit Acceleration=false)
   &ScaleSpriteSizeBySpeed(Max Scale Factor=<2 6>)
   /SetVariables(
    Particles.Render2COlor:LinearColor=&ColorFromCurve(ColorCurve=DataInterface() {propertyValues {
       AlphaCurve {keys [<at 0.1717 1>,<at 1 0>]}
       BlueCurve {keys [<at 0.2625 1>,<at 0.4663 0.7566>,<at 0.7194 0.8186>,<at 0.8822 0.7352>]}
       GreenCurve {keys [<at 0.2625 1>,<at 0.4663 0.7823>,<at 0.7194 0.8206>,<at 0.8822 0.7545>]}
       RedCurve {keys [<at 0.2625 1>,<at 0.4663 0.849>,<at 0.7194 0.8229>,<at 0.8822 0.8021>]}
      }})
    Particles.Render2SpriteSize:Vector2f=&Multiply_Vector2D(
     A=$Particles.Initial.Render2SpriteSize
     B=&Vector2DFromFloat(Value=&FloatFromCurve(FloatCurve=DataInterface() {propertyValues {Curve {keys [
           {tangentMode=RCTM_User, time=0, value=0.45}
           {interpMode=RCIM_Cubic, leaveTangent=0.7924, tangentMode=RCTM_Break, time=0.2, value=0.8}
           {tangentMode=RCTM_User, time=1, value=1}
          ]}}}))
    )
   ) {isSetParameters=true, module=SetVariables_#4}
   &CameraOffset(Camera Offset Amount=2)
  ]
  rendererClasses [&NiagaraSpriteRendererProperties]
  renderers [
   {rendererClass=&NiagaraSpriteRendererProperties}
   {rendererClass=&NiagaraSpriteRendererProperties, rendererIndex=1}
  ]
 }
 Splash001 {
  emitterUpdateScript [
   &EmitterState()
   &SpawnBurst_Instantaneous(Use Spawn Probability=false, UseLoopCountLimit=false)
  ]
  particleSpawnScript [
   &InitializeParticle(
    Lifetime Max=0.45
    Lifetime Min=0.3
    Lifetime Mode=::NewEnumerator1
    Position Offset=<0 0 25>
    Sprite Size Mode=::NewEnumerator3
    Uniform Sprite Size=&FN_FrontEndFloatModifier(Gameplay=110, Front End=75)
    UsePositionOffset=false
    Write Material Random=true
   )
   &SkeletalMeshLocation(Bone Sampling Mode=::NewEnumerator4, MaskSpawnGroup=false, Sampled Position Offset=$User.ParticleOffset, Skeletal Mesh=$User.SkeletalMesh)
   &Color(Color=&MultiplyLinearColor(
     Linear Color=$Particles.Initial.Color
     Scale Factor=&MakeVector4_FromLinearColor(LinearColor=&FN_Select_Color_from_Time_Of_Day(Morning=<rgba 1 1 1 1>, Day=<rgba 1 1 1 1>, Evening=<rgba 1 0.7676 0.7377 1>, Night=<rgba 0.9141 0.8046 1.25 1>))
    )) {enabled=false}
   /SetVariables(
    Particles.Render2COlor:LinearColor=<rgba 0.9219 0.9057 0.8979 1>
    Particles.Render2Rotation:NiagaraFloat=&RandomRangeFloat(Minimum=-360, Maximum=360)
    Particles.Render2SpriteSize:Vector2f=&Vector2DFromFloat(Value=&FN_FrontEndFloatModifier(Gameplay=140, Front End=125))
   ) {isSetParameters=true, module=SetVariables_#3}
   &Color(Color=&MultiplyLinearColor(
     Linear Color=$Particles.Render2COlor
     Scale Factor=&MakeVector4_FromLinearColor(LinearColor=&FN_Select_Color_from_Time_Of_Day(Morning=<rgba 1 1 1 1>, Day=<rgba 1 1 1 1>, Evening=<rgba 1 0.7676 0.7377 1>, Night=<rgba 0.7313 0.6437 1 1>))
    )) {module=Color001}
  ]
  particleUpdateScript [
   &ParticleState()
   &ScaleSpriteSize(Uniform Curve Sprite Scale=DataInterface() {propertyValues {Curve {keys [
        {tangentMode=RCTM_User, time=0, value=0.6}
        {interpMode=RCIM_Cubic, leaveTangent=0.5568, tangentMode=RCTM_Break, time=0.2, value=0.9}
        {tangentMode=RCTM_User, time=1, value=1}
       ]}}})
   &ScaleColor(
    Linear Color Curve=DataInterface() {propertyValues {
      AlphaCurve {keys [<at 0.4558 1>,<at 1 0>]}
      BlueCurve {keys [<at 0.3172 1>]}
      GreenCurve {keys [<at 0.3172 1>]}
      RedCurve {keys [<at 0.3172 1>]}
     }}
    Scale Mode=::NewEnumerator2
   )
   &GravityForce(Coordinate Space=::Simulation, Gravity=<0 0 -800>)
   &SolveForcesAndVelocity(Clamp Velocity=false, Limit Acceleration=false)
   &ScaleSpriteSizeBySpeed(Max Scale Factor=<2 6>)
   /SetVariables(
    Particles.Render2COlor:LinearColor=&ColorFromCurve(ColorCurve=DataInterface() {propertyValues {
       AlphaCurve {keys [<at 0.1717 1>,<at 1 0>]}
       BlueCurve {keys [<at 0.2625 1>,<at 0.4663 0.7566>,<at 0.7194 0.8186>,<at 0.8822 0.7352>]}
       GreenCurve {keys [<at 0.2625 1>,<at 0.4663 0.7823>,<at 0.7194 0.8206>,<at 0.8822 0.7545>]}
       RedCurve {keys [<at 0.2625 1>,<at 0.4663 0.849>,<at 0.7194 0.8229>,<at 0.8822 0.8021>]}
      }})
    Particles.Render2SpriteSize:Vector2f=&Multiply_Vector2D(
     A=$Particles.Initial.Render2SpriteSize
     B=&Vector2DFromFloat(Value=&FloatFromCurve(FloatCurve=DataInterface() {propertyValues {Curve {keys [
           {tangentMode=RCTM_User, time=0, value=0.45}
           {interpMode=RCIM_Cubic, leaveTangent=0.7924, tangentMode=RCTM_Break, time=0.2, value=0.8}
           {tangentMode=RCTM_User, time=1, value=1}
          ]}}}))
    )
   ) {isSetParameters=true, module=SetVariables_#4}
   &CameraOffset(Camera Offset Amount=2)
  ]
  rendererClasses [&NiagaraSpriteRendererProperties]
  renderers [
   {rendererClass=&NiagaraSpriteRendererProperties}
   {rendererClass=&NiagaraSpriteRendererProperties, rendererIndex=1}
  ]
 }
}
summary {
 systemName=NS_Glider_LilSplit_Deploy
 emitters [
  {emitterName=Empty003, rendererClasses [&NiagaraSpriteRendererProperties, &NiagaraMeshRendererProperties]}
  {emitterName=Empty005, rendererClasses [&NiagaraSpriteRendererProperties]}
  {emitterName=Empty006, rendererClasses [&NiagaraMeshRendererProperties]}
  {emitterName=Splash001, rendererClasses [&NiagaraSpriteRendererProperties]}
  {emitterName=Empty, rendererClasses [&NiagaraSpriteRendererProperties, &NiagaraMeshRendererProperties]}
  {emitterName=Empty001, rendererClasses [&NiagaraSpriteRendererProperties]}
  {emitterName=Empty002, rendererClasses [&NiagaraMeshRendererProperties]}
  {emitterName=Splash, rendererClasses [&NiagaraSpriteRendererProperties]}
  {emitterName=Empty004, rendererClasses [&NiagaraSpriteRendererProperties, &NiagaraMeshRendererProperties]}
 ]
 userVariables [
  User.AnimationDelay {type=NiagaraFloat, defaultValue=2.15, description=""}
  User.bIsFrontEnd {type=NiagaraBool}
  User.bIsFrontEndPreview {type=NiagaraBool}
  User.ForwardDot {type=NiagaraFloat, defaultValue=1, description=""}
  User.ParticleOffset {type=Vector3f, description="", defaultValue {struct=Vector3f, value=<20 -10 5>}}
  User.ParticleOffsetRight {type=Vector3f, description="", defaultValue {struct=Vector3f, value=<20 13 15>}}
  User.SkeletalMesh {
   type=NiagaraDataInterfaceSkeletalMesh
   defaultValue {
    struct=NiagaraExt_VariableValue_DataInterface
    value {dataInterface=&self:NiagaraDataInterfaceSkeletalMesh_0}
   }
  }
  User.VariantID {type=NiagaraFloat, defaultValue=1, description=""}
 ]
}
